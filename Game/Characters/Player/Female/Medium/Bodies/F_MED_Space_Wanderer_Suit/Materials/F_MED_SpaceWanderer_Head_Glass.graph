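# Translucent, surface-lit glass material for the Space Wanderer female character's head, showing animated green holographic overlays.
# Panning text modulated by scanlines combines with a parallax BumpOffset hex grid boosted by an edge mask, all scaled by Fresnel.
# A time-looped SphereMask over the Fresnel value sends a brightening pulse through the emissive.
# Opacity takes the maximum of text and Fresnel-hex terms, and the attributes pass through MF_TechArtOverrides.
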Material /Game/Characters/Player/Female/Medium/Bodies/F_MED_Space_Wanderer_Suit/Materials/F_MED_SpaceWanderer_Head_Glass
graph {
 properties {
  bUseMaterialAttributes=true
  bUsedWithMeshDeformer=true
  bUsedWithSkeletalMesh=true
  blendMode=BLEND_Translucent
  pixelDepthOffsetMode=PDOM_Legacy
  translucencyLightingMode=TLM_SurfacePerPixelLighting
  usedShadingModels=Default Lit
 }
 nodes [
  Constant()
  n1 = Panner(Coordinate=TextureCoordinate() {coordinateIndex=1}) {speedX=0.05}
  n2 = TextureSample(UVs=TextureCoordinate() {coordinateIndex=1}) {texture=/Game/Characters/Player/Female/Medium/Bodies/F_MED_Space_Wanderer/Textures/Panningtextmask}
  n3 = Add(
   A=Multiply(A=TextureSample(UVs=@n1) {texture=&Panningtextmask}.R, B=@n2.G)
   B=@n2.B
  )
  n4 = Multiply(A=VectorParameter() {parameterName=Green Text, defaultValue=<rgba 0.0553 0.3281 0 1>}.RGB, B=@n3)
  n5 = MaterialFunctionCall!(ScanLine_Tiling_1, Scanline_PannerSpeed_1, ScanLine_Tiling, Scanline_PannerSpeed, UVs=TextureCoordinate() {coordinateIndex=1}) {materialFunction=/Game/Weapons/WeaponSkins/MaterialFunction/ScanLines}
  n6 = TextureCoordinate()
  n7 = ScalarParameter() {defaultValue=10, parameterName=Hex Tiling}
  n8 = TextureSample(UVs=BumpOffset(Coordinate=Multiply(A=@n6, B=@n7), Height=ScalarParameter() {defaultValue=-2, parameterName=Hex offset Height})) {samplerType=SAMPLERTYPE_Masks, texture=/Game/Accessories/FORT_Backpacks/F_MED_CyberRunner_Backpack/Textures/HexGrid_M}
  n9 = Multiply(A=@n8.R, B=VectorParameter() {parameterName=Head expansion, defaultValue=<rgba 0.1037 0.349 0 1>}.RGB)
  n10 = Multiply(A=@n9) {constB=1.5}
  n11 = TextureSampleParameter2D(UVs=StaticSwitchParameter(True=TextureCoordinate(), False=TextureCoordinate() {coordinateIndex=1}) {parameterName=EdgeMaskuseFirstUV}) {
   parameterName=Edge Mask
   samplerType=SAMPLERTYPE_LinearColor
   texture=/Game/Characters/Player/Female/Medium/Bodies/F_MED_Space_Wanderer/Textures/EdgeMask
  }
  n12 = ScalarParameter() {defaultValue=2.75, parameterName=Fresnel Exponent}
  n13 = TextureSampleParameter2D() {
   group=Mask Texture
   parameterName=BakedNormal
   samplerSource=SSM_Wrap_WorldGroupSettings
   samplerType=SAMPLERTYPE_Normal
   texture=/Engine/EngineMaterials/FlatNormal
  }
  n14 = Normalize(VectorInput=@n13.RGB)
  n15 = Transform(None=@n14)
  n16 = Clamp(None=Multiply(A=Fresnel(ExponentIn=@n12, Normal=@n15), B=ScalarParameter() {defaultValue=1, parameterName=Glass thickness}))
  n17 = Multiply(
   A=Add(A=Multiply(A=@n4, B=@n5.Result), B=LinearInterpolate(A=@n9, B=@n10, Alpha=@n11.R))
   B=LinearInterpolate(Alpha=@n16) {constA=0.075, constB=0.95}
  )
  n18 = ScalarParameter() {defaultValue=0.33, parameterName=Emissive Pulse Thickness}
  n19 = Time()
  n20 = ScalarParameter() {defaultValue=0.33, parameterName=Time Speed}
  n21 = LinearInterpolate(A=Subtract(B=@n18) {constA=0}, B=Add(B=@n18) {constA=1.5}, Alpha=Frac(None=Multiply(A=@n19, B=@n20)))
  n22 = Constant()
  n23 = LinearInterpolate(Alpha=@n16) {constA=0.25}
  n24 = LinearInterpolate(
   A=ScalarParameter() {parameterName=Fresnel Inner Transparency}
   B=ScalarParameter() {defaultValue=0.5, parameterName=Fresnel Outer Transparency}
   Alpha=@n16
  )
  n25 = Multiply(A=@n11.R, B=@n8.R)
  n26 = SetMaterialAttributes!(
   MaterialAttributes
   Base Color=VectorParameter() {parameterName=Basic Tint, defaultValue=<rgba 0.3009 0.6146 0.1347 1>}.RGB
   Metallic
   Specular=ScalarParameter() {defaultValue=1, parameterName=Specular Glass}
   Roughness=ScalarParameter() {defaultValue=0.15, parameterName=Roughness Glass}
   Anisotropy
   Emissive Color=LinearInterpolate(A=@n17, B=Multiply(A=@n17) {constB=5}, Alpha=SphereMask(A=@n21, B=@n16, Radius=@n18, Hardness=@n22))
   Opacity=Max(A=Multiply(A=@n3, B=@n23), B=LinearInterpolate(A=@n24, Alpha=@n25) {constB=0.25})
   Opacity Mask
   Normal=@n14
   Tangent
  ) {attributeSetTypes [
    #1
    #2
    #3
    #4
    #5
    #6
    #7
    #8
    #9
    #10
   ]}
  MaterialFunctionCall(In=@n26) {materialFunction=&MF_TechArtOverrides}
  Multiply()
  n27 = Floor(None=ScalarParameter() {parameterName=Use Pre-Skinned Bounds, sliderMax=1})
  n28 = Divide(A=PreSkinnedPosition(), B=LinearInterpolate(A=ObjectBounds(), B=PreSkinnedLocalBounds().Half Extents, Alpha=Saturate(None=@n27)))
  VertexInterpolator(VS=Multiply(A=ComponentMask(None=@n28) {b=true, g=true}, B=ScalarParameter() {defaultValue=0.2, parameterName=WaterTiling}))
 ]
}
